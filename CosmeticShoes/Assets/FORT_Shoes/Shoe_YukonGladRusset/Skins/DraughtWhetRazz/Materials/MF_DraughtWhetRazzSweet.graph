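# Material function for the DraughtWhetRazz shoe skin that layers a glowing, Slurp-derived two-color "Sweet" look onto incoming material attributes.
# Emissive blends tinted, hue-shifted colors with inner and back scattering from sunlight direction, using masks built from pre-skinned position.
# A custom spherical-harmonics evaluation decodes vertex-color coefficients along a skinning-aware view direction into a per-vertex mask.
# Projected noise adds sugar grain before time-of-day color adjustment, while metallic is zeroed and normals optionally flattened.
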MaterialFunction /CosmeticShoes/Assets/FORT_Shoes/Shoe_YukonGladRusset/Skins/DraughtWhetRazz/Materials/MF_DraughtWhetRazzSweet
graph {nodes [
  n1 = VectorParameter() {group=Sweet, parameterName=Sweet_ColorA, defaultValue=<rgba 0.6562 0.0261 0.0065 1>}
  n2 = VectorParameter() {group=Sweet, parameterName=Sweet_ColorB, defaultValue=<rgba 0 0.625 0 1>}
  n3 = ScalarParameter() {defaultValue=1.2, group=Sweet, parameterName=Sweet_OuterColorMult}
  n4 = Constant() {r=0.4}
  n5 = Saturate(None=Add(A=Multiply(A=@n1.RGB, B=@n3) {constB=0.2}, B=@n4) {constB=0.2})
  n6 = Saturate(None=Add(A=Multiply(A=@n2.RGB, B=@n3) {constB=0.2}, B=@n4) {constB=0.2})
  n7 = ScalarParameter() {defaultValue=0.1, group=Sweet, parameterName=Sweet_OuterColorHueShiftAmount}
  n8 = Subtract(
   A=ComponentMask(None=NamedRerouteUsage()) {b=true}
   B=ScalarParameter() {defaultValue=125, group=Sweet, parameterName=Sweet_Slurp_OuterColor_TransitionStart}
  ) {constB=125}
  n9 = ScalarParameter() {defaultValue=0.015, group=Sweet, parameterName=Sweet_Slurp_OuterColor_TransitionScale}
  n10 = LinearInterpolate(
   A=LinearInterpolate(A=@n5, B=@n6, Alpha=NamedRerouteUsage())
   B=LinearInterpolate(
    A=MaterialFunctionCall(Hue Shift Percentage=@n7, Texture=@n5) {materialFunction=&HueShift}.Result
    B=MaterialFunctionCall(Hue Shift Percentage=@n7, Texture=@n6) {materialFunction=&HueShift}.Result
    Alpha=NamedRerouteUsage()
   )
   Alpha=Saturate(None=Multiply(A=@n8, B=@n9) {constB=0.015})
  )
  n11 = LinearInterpolate(A=LinearInterpolate(A=@n1.RGB, B=@n2.RGB, Alpha=NamedRerouteUsage()), B=@n10, Alpha=NamedRerouteUsage())
  n12 = VectorParameter() {group=Sweet, parameterName=Sweet_SlurpEmissiveInnerScatteringColorMultiplier, defaultValue=<rgba 1 1 1 0>}
  n13 = Multiply(B=@n12.RGB)
  n14 = Multiply(B=@n12.RGB)
  n15 = StaticSwitchParameter(
   True=StaticSwitchParameter(True=LinearInterpolate(A=@n13, B=@n14), False=@n13) {group=Sweet, parameterName=Sweet_SlurpUseCustomEmissiveAsInnerScatteringColor}
   False=VectorParameter() {group=Sweet, parameterName=Sweet_Slurp_InnerScatteringColor, defaultValue=<rgba 1 0 0.016 1>}.RGB
  ) {group=Sweet, parameterName=Sweet_SlurpUseEmissiveAsInnerScatteringColor}
  n16 = NamedRerouteUsage()
  n17 = Constant3Vector() {constant=<rgba 1 1 1 0>}
  n18 = VectorParameter() {group=Slurp, parameterName=Slurp_InnerBackScatteringMultiplier, defaultValue=<rgba 0 0 0 1>}
  n19 = NamedRerouteUsage()
  n20 = LinearInterpolate(
   A=Multiply(A=LinearInterpolate(A=@n11, B=@n15, Alpha=@n16), B=LinearInterpolate(A=@n17, B=@n18.RGB, Alpha=@n19))
   B=Constant()
   Alpha=NamedRerouteUsage()
  )
  n21 = NamedRerouteUsage()
  n22 = Noise!(World Position=Multiply(A=@n21) {constB=0.5}, FilterWidth) {levels=1}
  n23 = ConstantBiasScale(None=@n22)
  n24 = MaterialFunctionCall!(
   TextureObject=TextureObject() {texture=&AtmosphericCloudNoise01}
   TextureSize=ScalarParameter() {defaultValue=2.5, group=Sweet, parameterName=Sweet_NoiseScale}
   WorldPosition=Add(A=Multiply(A=@n22) {constB=0.15}, B=@n21)
   Export Float 4
   In
   World Space Normal=NamedRerouteUsage()
   ProjectionTransitionContrast=Constant() {r=10}
  ) {materialFunction=/Game/Athena/SafeZone/Preskinned_WP_Texture_ScalePan}
  n25 = Multiply(A=Max(A=@n23) {constB=0.9}, B=@n24.XYZ Texture)
  n26 = StaticComponentMaskParameter(None=Saturate(None=@n25)) {defaultR=true, parameterName=SweetFXChannel}
  n27 = Power(Base=@n26) {constExponent=1.4}
  n28 = Add(A=@n20, B=Multiply(A=@n27, B=ScalarParameter() {defaultValue=0.05, group=Sweet, parameterName=SugarStrength}) {constB=0.1})
  n29 = Power(Base=@n26)
  Add(A=@n28, B=Multiply(A=Multiply(A=@n29, B=@n29)) {constB=0.8})
  Add(A=@n23, B=@n24.XYZ Texture)
  Add(A=@n25, B=@n24.XYZ Texture)
  n30 = ScalarParameter() {defaultValue=0.4, parameterName=Sweet_BlendMin}
  Add(A=@n30) {constB=0}
  n31 = VectorParameter() {group=Sweet, parameterName=Sweet_SHMax, defaultValue=<rgba 1.25 0.13 0.18 0.13>}
  n32 = VectorParameter() {group=Sweet, parameterName=Sweet_SHMin, defaultValue=<rgba 1 -0.16 -0.17 -0.15>}
  n33 = AppendVector(A=@n32.RGB, B=@n32.RGB)
  n34 = Subtract(A=AppendVector(A=@n31.RGB, B=@n31.RGB), B=@n33)
  n35 = VertexColor()
  n36 = AppendVector(A=@n35, B=@n35)
  n37 = Multiply(A=MaterialFunctionCall(X=@n35.G, Y=@n35.G, Z=@n35.G) {materialFunction=&MakeFloat3}.Result)
  MaterialExpressionStaticSwitchParameter_8 = StaticSwitchParameter(True=Add(A=Multiply(A=@n34, B=@n36), B=@n33), False=AppendVector(A=Add(A=@n35.R), B=Add(A=@n37) {constB=-0.5})) {group=Sweet, parameterName=Sweet_UseCustomSH Fit}
  n38 = MaterialFunctionCall() {materialFunction=&CameraDirectionVector}
  n39 = ComponentMask(None=Multiply(A=@n38.Result, B=ScalarParameter() {defaultValue=1, group=Slurp, parameterName=Slurp_DirectionVectorFlip})) {b=true, g=true, r=true}
  n40 = VectorParameter() {group=Slurp, parameterName=SlurpNormalSHVectorDistortion}
  n41 = StaticSwitch(
   True=Add(A=Multiply(B=@n40.RGB), B=@n39)
   False=@n39
   Value=StaticBoolParameter() {group=Sweet, parameterName=Sweet_UseSlurpNormalDistortion}
  )
  n42 = StaticBoolParameter() {group=Slurp, parameterName=UseTriPlanarBubbles}
  n43 = Multiply(A=Reroute())
  n44 = Reroute()
  n45 = LinearInterpolate(
   A=@n41
   B=StaticSwitch(True=Transform() {transformSourceType=TRANSFORMSOURCE_Local}, False=Transform(), Value=@n42)
   Alpha=Multiply(A=StaticSwitch(True=@n43, False=@n44, Value=@n42), B=ScalarParameter() {defaultValue=0.25, group=Sweet, parameterName=Sweet_Bubble_SHScale})
  )
  n46 = Transform(None=StaticSwitch(True=Normalize(VectorInput=@n45), False=@n41, Value=Reroute())) {transformSourceType=TRANSFORMSOURCE_World, transformType=TRANSFORM_Local}
  n47 = VectorParameter() {group=Sweet, parameterName=Sweet_Slurp_CustomDirectionVector, defaultValue=<rgba 0 0 -1 1>}
  n48 = ScalarParameter() {defaultValue=1, group=Sweet, parameterName=Sweet_CustomDirection_Bias}
  n49 = StaticSwitchParameter(True=Normalize(VectorInput=LinearInterpolate(A=@n46, B=@n47.RGB, Alpha=@n48)), False=@n46) {group=Sweet, parameterName=Sweet_UseCustomDirection}
  n50 = VertexNormalWS()
  MaterialExpressionStaticSwitchParameter_1 = StaticSwitchParameter(
   True=MaterialFunctionCall!(
    VectorToRotate=@n49
    PreSkinnedLocalNormal=NamedRerouteUsage()
    VertexlNormalWS=Transform(None=@n50) {transformSourceType=TRANSFORMSOURCE_World, transformType=TRANSFORM_Local}
   ) {materialFunction=/Game/Packages/Fortress_SharedMaterials/Base_Material_Functions/RotateVector_FromSkinning}.RotatedVectorFromSkinning
   False=@n49
  ) {defaultValue=true, group=Sweet, parameterName=Sweet_UseRotateVector_FromSkinning_SH_Dir}
  n51 = Custom!(SH=@MaterialExpressionStaticSwitchParameter_8, Direction=@MaterialExpressionStaticSwitchParameter_1) {
   code="//SH decode,  Direction = LightDir, SH = 2band coefficients. (encoded in Y-up)\r\nDirection = normalize(Direction);\r\nfloat4 shEvalDirection;\r\nshEvalDirection.x = 0.282095;\r\nshEvalDirection.y =0.488603 * Direction.z;\r\nshEvalDirection.z = 0.488603 * Direction.y;\r\nshEvalDirection.w =-0.488603 * Direction.x;\r\nreturn saturate(dot(SH, shEvalDirection));"
   description=SphericalHarmonicsEval
   outputType=CMOT_Float1
   inputs [
    {
     inputName=SH
     input {expression=&self:MaterialExpressionStaticSwitchParameter_8}
    }
    {
     inputName=Direction
     input {expression=&self:MaterialExpressionStaticSwitchParameter_1}
    }
   ]
  }
  n52 = OneMinus(None=MaterialFunctionCall(
    Input=@n51
    Input Low=ScalarParameter() {group=Sweet, parameterName=Sweet_SHContrast_LowValue}
    Input High=ScalarParameter() {defaultValue=1, group=Sweet, parameterName=Sweet_SHContrast_HighValue}
    Target Low=Constant()
    Target High=Constant() {r=1}
   ) {materialFunction=&RemapValueRange}.Result)
  n53 = VertexInterpolator(VS=Power(Base=Saturate(None=@n52), Exp=ScalarParameter() {defaultValue=1, group=Sweet, parameterName=Sweet_SHPow}))
  n54 = Saturate(None=ComponentMask(None=MaterialFunctionCall() {materialFunction=/Engine/Functions/Engine_MaterialFunctions02/Texturing/3DSandMayaUVCoordinates}.UVChannel1) {r=true})
  Add(A=@n53.PS, B=Multiply(A=OneMinus(None=@n54), B=ScalarParameter() {group=Sweet, parameterName=Sweet_Slurp_UV1_Darkening}) {constB=0.0333})
  n55 = ScalarParameter() {defaultValue=0.7, parameterName=Sweet_BlendMax}
  Add(A=@n55) {constB=0}
  Constant()
  Constant() {r=0.2}
  Constant() {r=0.5}
  n56 = GetMaterialAttributes(None=FunctionInput() {inputType=FunctionInput_MaterialAttributes}) {attributeGetTypes [
    #1
    #2
    #3
    #4
    #5
   ]}
  n57 = SetMaterialAttributes!(
   MaterialAttributes=@n56.MaterialAttributes
   BaseColor
   EmissiveColor=MaterialFunctionCall!(Color=@n28) {materialFunction=/Game/Packages/Fortress_SharedMaterials/Anime_Master_Material/MF_TODColorAdjustment}.Result
   Metallic=Constant()
   Specular
   Roughness
   Normal=MaterialFunctionCall(Normal=@n56.Normal, Flatness=ScalarParameter() {group=Sweet, parameterName=Sweet_NormalFlatness}) {materialFunction=&FlattenNormal}.Result
  ) {attributeSetTypes [
    #1
    #6
    #2
    #3
    #4
    #5
   ]}
  FunctionOutput(None=@n57)
  LinearInterpolate(A=@n10)
  MaterialFunctionCall() {materialFunction=&CameraDirectionVector}
  n58 = Desaturation(None=@n56.BaseColor)
  n59 = ScalarParameter() {defaultValue=0.2, group=Sweet, parameterName=Sweet_BaseColorIntensity}
  Max(A=Constant() {r=0.015}, B=Add(A=LinearInterpolate(B=@n58, Alpha=@n59), B=@n56.Specular))
  n60 = ScalarParameter() {defaultValue=1, group=Sweet, parameterName=Sweet_EmissiveMult}
  Multiply(
   A=@n20
   B=MaterialFunctionCall(
    Value_Morning=@n60
    Value_Day=@n60
    Value_Evening=ScalarParameter() {defaultValue=0.5, group=Sweet, parameterName=Sweet_EmissiveMult_Night}
    Value_Night=@n60
   ) {materialFunction=&Scalar_Time_of_Day}.Result
  )
  n61 = QualitySwitch(Default=@n53.PS, Low=@n53.PS)
  NamedRerouteDeclaration(None=ShadingPathSwitch(Default=@n61, Deferred=@n61, Forward=@n53.PS, Mobile=@n53.PS)) {name=SH Mask, nodeColor=<rgba 0.5938 0 1 1>}
  n62 = CollectionParameter() {
   collection=&FortniteMaterialParameters
   parameterName=SunlightDirectionalVector
  }
  n63 = ComponentMask(None=@n62) {b=true, g=true, r=true}
  n64 = StaticSwitch(False=PixelNormalWS(), Value=StaticBoolParameter() {group=Slurp, parameterName=UseSlurpNormalDistortion})
  n65 = Multiply(A=Multiply(A=DotProduct(A=@n63, B=@n64)) {constB=2}, B=@n64)
  n66 = DotProduct(
   A=Normalize(VectorInput=Subtract(A=@n65, B=@n63))
   B=Multiply(A=MaterialFunctionCall() {materialFunction=&CameraDirectionVector}.Result) {constB=-1}
  )
  n67 = Power(Base=Max(A=@n66) {constB=0}, Exp=ScalarParameter() {defaultValue=1, group=Slurp, parameterName=Slurp_InnerScatteringPower})
  n68 = ScalarParameter() {defaultValue=0.4, group=Slurp, parameterName=Slurp_InnerScatteringScale}
  NamedRerouteDeclaration(None=Saturate(None=Multiply(A=@n67, B=@n68) {constB=0.4})) {name=Scattering Mask, nodeColor=<rgba 0 0.1797 1 1>}
  n69 = PreSkinnedPosition()
  n70 = FunctionInput(Preview=ShaderStageSwitch(PixelShader=VertexInterpolator(VS=@n69).PS, VertexShader=@n69)) {bUsePreviewValueAsDefault=true, inputName=Pre Skinned Position}
  n71 = PreSkinnedLocalBounds()
  n72 = ComponentMask(None=Divide(A=NamedRerouteDeclaration(None=@n70) {name=PreSkinnedPos_VS-PS, nodeColor=<rgba 0 0.4844 1 1>}, B=Subtract(A=@n71.Max, B=@n71.Max))) {b=true}
  n73 = Saturate(None=SmoothStep(Min=@n30, Max=@n55, Value=@n72) {constMax=0.7, constMin=0.4})
  NamedRerouteDeclaration(None=StaticSwitchParameter(True=OneMinus(None=@n73), False=Constant()) {defaultValue=true, parameterName=DualColors}) {name=Tint Mask, nodeColor=<rgba 0.6406 0 1 1>}
  n74 = Multiply(A=Subtract(A=@n55, B=@n30)) {constB=0.5}
  n75 = Add(A=Add(A=@n74, B=@n30)) {constB=0.15}
  n76 = Add(A=@n74) {constB=0.15}
  n77 = Constant()
  n78 = Add(
   A=Power(
    Base=SphereMask(A=@n72, B=@n75, Radius=@n76, Hardness=@n77)
    Exp=ScalarParameter() {defaultValue=4, parameterName=Sweet_DesatuationRangePower}
   ) {constExponent=4}
   B=ScalarParameter() {defaultValue=-0.5, parameterName=Sweet_DesatuationBias}
  ) {constB=-0.5}
  NamedRerouteDeclaration(None=StaticSwitchParameter(True=Saturate(None=@n78), False=Constant()) {defaultValue=true, parameterName=DualColors}) {name=Desatuation Mask, nodeColor=<rgba 0 1 0.5547 1>}
  n79 = Multiply(A=@n66) {constB=-1}
  n80 = ScalarParameter() {group=Slurp, parameterName=Sweet_Slurp_InnerBackScatteringBias}
  n81 = Power(
   Base=Max(A=Add(A=@n79, B=@n80)) {constB=0}
   Exp=ScalarParameter() {defaultValue=1, group=Slurp, parameterName=Sweet_Slurp_InnerBackScatteringPower}
  )
  n82 = ScalarParameter() {defaultValue=0.4, group=Slurp, parameterName=Sweet_Slurp_InnerBackScatteringScale}
  NamedRerouteDeclaration(None=Saturate(None=Multiply(A=@n81, B=@n82) {constB=0.55})) {name=BackScatteringMask, nodeColor=<rgba 0 0.7656 1 1>}
  n83 = PreSkinnedNormal()
  NamedRerouteDeclaration(None=ShaderStageSwitch(PixelShader=VertexInterpolator(VS=@n83).PS, VertexShader=@n83)) {name=PreSkinnedNrm_VS-PS, nodeColor=<rgba 0 0.4844 1 1>}
  ScalarParameter() {defaultValue=10, parameterName=asds}
  SmoothStep(Value=@n27) {constMax=0.5, constMin=0.1}
  n84 = Multiply(A=@n29) {constB=0.75}
  SmoothStep(Value=Saturate(None=Add(A=@n56.Specular, B=@n84))) {constMax=0.9, constMin=0.2}
  StaticSwitchParameter(False=@n38.Result) {group=Slurp, parameterName=UseSunDirection}
  Transform(None=CameraVectorWS()) {transformSourceType=TRANSFORMSOURCE_World, transformType=TRANSFORM_Local}
  Transform(None=Constant3Vector() {constant=<rgba 0 0 1 0>}) {transformSourceType=TRANSFORMSOURCE_View, transformType=TRANSFORM_Local}
  VectorParameter() {parameterName=Inner Color, defaultValue=<rgba 0 0 0 1>}
 ]}
